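# One-shot stereo MetaSound source for the CubeSurfing surf-ready cue, apparently first-person, using a recharge wave in the melee weapon sound class.
# On play, a shuffle picks from the Sounds array, which holds one wave by default, and triggers the wave player.
# Base-and-spread inputs randomize pitch shift and gain; gain passes through a Max node and scales both channels.
# The player's finish trigger ends the source.
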MetaSoundSource /CubeSurfing/Assets/Audio/MSS_CubeSurfing_Surf_Ready_1P
inputs [
 Random.Pitch.Base {type=Float}
 Random.Pitch.Spread {type=Float}
 Random.Volume.Base {type=Float}
 Random.Volume.Spread {type=Float}
 Sounds {type=WaveAsset:Array}
 UE.Source.OnPlay {type=Trigger}
]
interfaces [UE.OutputFormat.Stereo:1.0, UE.Source.OneShot:1.0, UE.Source:1.1]
outputs [
 "UE.OutputFormat.Stereo.Audio:0" {type=Audio}
 "UE.OutputFormat.Stereo.Audio:1" {type=Audio}
 UE.Source.OneShot.OnFinished {type=Trigger}
]
pages [Default {
  input_defaults {
   Random.Pitch.Base=0
   Random.Pitch.Spread=1
   Random.Volume.Base=1
   Random.Volume.Spread=0
   UE.Source.OnPlay=null
   Sounds [/CubeSurfing/Assets/Audio/Waves/Ability_CubeSurfing_Surf_Recharge_01]
  }
  nodes [
   n1 = Max.Max:Float(A=UE.RandomFloat(
     Max=UE.Add:Float(PrimaryOperand=^Random.Volume.Base, AdditionalOperands=^Random.Volume.Spread).Out
     Next=UE.Reroute(Value=^UE.Source.OnPlay).Value
     Min=UE.Subtract:Float(PrimaryOperand=^Random.Volume.Base, AdditionalOperands=^Random.Volume.Spread).Out
    ).Value)
   n2 = Array.Shuffle:WaveAsset:Array(In Array=^Sounds, Next=^UE.Source.OnPlay) {also {defaults {Enable Shared State="<MetaSoundBuilderResult.SUCCEEDED: 0>"}}}
   WavePlayer_0 = UE.Wave Player:Stereo(
    Play=@n2.On Next
    Wave Asset=@n2.Value
    Pitch Shift=UE.RandomFloat(
     Max=UE.Add:Float(PrimaryOperand=^Random.Pitch.Base, AdditionalOperands=^Random.Pitch.Spread).Out
     Next=UE.Reroute(Value=^UE.Source.OnPlay).Value
     Min=UE.Subtract:Float(PrimaryOperand=^Random.Pitch.Base, AdditionalOperands=^Random.Pitch.Spread).Out
    ).Value
   )
   Multiply_0 = UE.Multiply:Audio by Float(AdditionalOperands=@n1.Value, PrimaryOperand=@WavePlayer_0.Out Left)
   Multiply_1 = UE.Multiply:Audio by Float(AdditionalOperands=@n1.Value, PrimaryOperand=@WavePlayer_0.Out Right)
  ]
  links [
   {dst="@UE.OutputFormat.Stereo.Audio:0", src=Multiply_0.Out}
   {dst="@UE.OutputFormat.Stereo.Audio:1", src=Multiply_1.Out}
   {dst="@UE.Source.OneShot.OnFinished", src=WavePlayer_0.On Finished}
  ]
 }]
properties {
 assetImportData=&self:AssetImportData
 numChannels=2
 outputFormat=Stereo
 sampleRate=48000
 soundClassObject=&Fort_WeaponMelee
 volume=0.4
}
