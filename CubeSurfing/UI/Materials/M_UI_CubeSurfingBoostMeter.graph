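# Translucent UI material for the CubeSurfing boost meter, drawn as a curved arc with cooldown and error states.
# An SDF circle outline is masked by a dimmed full arc combined with a radial fill from inverted MeterProgress, with anti-aliased edges.
# isCooldown and isError blend the color, and cooldown also changes opacity.
# Optional parallax offsets the UVs by scaled, clamped, negated CameraVelocity so the meter drifts with camera movement.
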Material /CubeSurfing/UI/Materials/M_UI_CubeSurfingBoostMeter
graph {
 properties {bAutomaticallySetUsageInEditor=false, blendMode=BLEND_Translucent, materialDomain=MD_UI, usedShadingModels=Default Lit}
 nodes [
  LinearInterpolate(B=NamedRerouteUsage(), Alpha=NamedRerouteUsage())
  n1 = MaterialFunctionCall(
   UVs=NamedRerouteUsage()
   Size=ScalarParameter() {defaultValue=0.85, group=Circle Parameters, parameterName=CircleSize, sliderMax=1, sortPriority=4}
   OutlineThickness=ScalarParameter() {defaultValue=0.02, group=Circle Parameters, parameterName=CircleThickness, sliderMax=1, sortPriority=4}
  ) {materialFunction=&MF_UI_Circle}
  n2 = ScalarParameter() {defaultValue=0.1765, group=Circle Parameters, parameterName=StartAngle, sliderMax=1, sortPriority=4}
  n3 = ScalarParameter() {defaultValue=0.681, group=Circle Parameters, parameterName=EndAngle, sliderMax=1, sortPriority=4}
  n4 = MaterialFunctionCall!(UV=NamedRerouteUsage(), "Arc Start Angle (0-1)"=@n2, "Arc End Angle (0-1)"=@n3, "Fill Ratio (0-1)"=Constant() {r=1}, Fill Clockwise) {materialFunction=/FortUILibrary/MaterialFunctions/Basics/MF_UI_RadialArc}
  n5 = Constant() {r=0.001}
  n6 = Saturate(None=Subtract(A=MaterialFunctionCall(DistanceField=@n4.Arc Mask, Threshold=@n5) {materialFunction=&MF_UI_AntiAliasing}.Outer) {constB=0.5})
  n7 = NamedRerouteUsage()
  n8 = OneMinus(None=ScalarParameter() {defaultValue=1, group=Progress Parameters, parameterName=MeterProgress, sliderMax=1, sortPriority=1})
  n9 = Constant()
  n10 = NamedRerouteDeclaration(None=MaterialFunctionCall(
    DistanceField=MaterialFunctionCall(UV=@n7, "Arc Start Angle (0-1)"=@n2, "Arc End Angle (0-1)"=@n3, "Fill Ratio (0-1)"=@n8, Fill Clockwise=@n9) {materialFunction=&MF_UI_RadialArc}.Radial Fill
    Threshold=@n5
   ) {materialFunction=&MF_UI_AntiAliasing}.Outer) {name=SegmentedMeterFill, nodeColor=<rgba 0.025 0.0203 0.0125 0.75>}
  Multiply(
   A=Min(
    A=MaterialFunctionCall(DistanceField=@n1.SDFOutline) {materialFunction=&MF_UI_AntiAliasing}.Inner
    B=Max(A=@n6, B=@n10)
   )
   B=LinearInterpolate(
    A=ScalarParameter() {defaultValue=0.7, parameterName=MasterOpacity, sliderMax=1}
    B=Constant() {r=0.4}
    Alpha=ScalarParameter() {parameterName=isCooldown}
   )
  )
  n11 = VectorParameter() {parameterName=ProgressColor, defaultValue=<rgba 0.875 0.875 0.875 1>}
  n12 = VectorParameter() {parameterName=CooldownColor, defaultValue=<rgba 0.2552 0.2552 0.2552 1>}
  n13 = ScalarParameter() {parameterName=isCooldown}
  NamedRerouteDeclaration(None=LinearInterpolate(
    A=LinearInterpolate(A=@n11.RGB, B=@n12.RGB, Alpha=@n13)
    B=VectorParameter() {parameterName=ErrorColor, defaultValue=<rgba 1 0 0.083 1>}.RGB
    Alpha=ScalarParameter() {parameterName=isError}
   )) {name=ProgressColor_1, nodeColor=<rgba 0.025 0.0203 0.0125 0.75>}
  n14 = Divide(
   A=ComponentMask(None=VectorParameter() {parameterName=CameraVelocity, defaultValue=<rgba 1 1 1 1>}.RGB) {g=true, r=true}
   B=Constant2Vector() {g=-10, r=-10}
  ) {desc="Decreases movement\r\nof element relative to others"}
  n15 = Constant2Vector() {desc=Max Drift, g=10, r=0.9}
  n16 = Multiply(A=@n15) {constB=-1}
  NamedRerouteDeclaration(None=Add(
    A=MaterialFunctionCall() {materialFunction=&GetUserInterfaceUV}.Normalized UV
    B=Clamp(None=@n14, Min=@n16, Max=@n15)
   )) {name=ParallaxUV_1, nodeColor=<rgba 0 0 0 1>}
  n17 = MaterialFunctionCall() {materialFunction=&GetUserInterfaceUV}
  n18 = NamedRerouteUsage()
  n19 = ScalarParameter() {parameterName=isParallax}
  NamedRerouteDeclaration(None=Add(A=LinearInterpolate(A=@n17.Normalized UV, B=@n18, Alpha=@n19), B=Constant2Vector() {r=0.2})) {name=ModUvs, nodeColor=<rgba 0.5 0.4065 0.25 0.75>}
 ]
}
