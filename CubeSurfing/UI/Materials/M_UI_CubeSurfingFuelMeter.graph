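# Translucent UI material for the CubeSurfing fuel meter, drawn as a segmented arc on an anti-aliased SDF circle outline.
# Ammo-named parameters drive the fill from ActiveAmmo over ClipSize, switching to ReloadProgress while that is above zero.
# Radial arc functions build fill, delta and background masks, while a ping-pong gradient forms segment gaps.
# ProgressColor and DeltaColor supply tints, isCooldown blends opacity, and isParallax enables clamped UV drift from CameraVelocity.
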Material /CubeSurfing/UI/Materials/M_UI_CubeSurfingFuelMeter
graph {
 properties {bAutomaticallySetUsageInEditor=false, blendMode=BLEND_Translucent, materialDomain=MD_UI, usedShadingModels=Default Lit}
 nodes [
  n1 = NamedRerouteUsage()
  LinearInterpolate(A=LinearInterpolate(B=NamedRerouteUsage(), Alpha=@n1), B=NamedRerouteUsage(), Alpha=@n1)
  n2 = NamedRerouteUsage()
  n3 = ScalarParameter() {defaultValue=0.181, group=Circle Parameters, parameterName=StartAngle, sliderMax=1, sortPriority=4}
  n4 = ScalarParameter() {defaultValue=0.689, group=Circle Parameters, parameterName=EndAngle, sliderMax=1, sortPriority=4}
  n5 = ScalarParameter() {defaultValue=50, group=Ammo Parameters, parameterName=ClipSize, sliderMax=100, sortPriority=6}
  n6 = ScalarParameter() {group=Ammo Parameters, parameterName=ReloadProgress, sliderMax=1, sortPriority=6}
  n7 = Ceil(None=MaterialFunctionCall!(A=@n6, B=Constant()) {materialFunction=/FortUILibrary/MaterialFunctions/Basics/MF_UI_BooleanOperations}.IsGreater) {desc=Controls Switches}
  n8 = Floor(None=LinearInterpolate(A=@n5, B=Constant() {desc=Reload Segment Steps, r=1000}, Alpha=@n7) {desc=Switch for Segment Count})
  n9 = LinearInterpolate(
   A=Divide(A=ScalarParameter() {defaultValue=50, group=Ammo Parameters, parameterName=ActiveAmmo, sliderMax=100, sortPriority=6}, B=@n5)
   B=@n6
   Alpha=@n7
  ) {constA=0.1, constB=0, desc=Switch for Meter Progress}
  n10 = StaticSwitch(
   True=MaterialFunctionCall!("Value (0-1)"=@n9, Steps=@n8) {materialFunction=/FortUILibrary/MaterialFunctions/Basics/MF_UI_StepRatio}.Stepped Ratio
   False=@n9
   Value=StaticBool()
  )
  n11 = OneMinus(None=@n10)
  n12 = Constant()
  n13 = Constant() {r=0.001}
  n14 = NamedRerouteDeclaration(None=MaterialFunctionCall(
    DistanceField=MaterialFunctionCall!(UV=@n2, "Arc Start Angle (0-1)"=@n3, "Arc End Angle (0-1)"=@n4, "Fill Ratio (0-1)"=@n11, Fill Clockwise=@n12) {materialFunction=/FortUILibrary/MaterialFunctions/Basics/MF_UI_RadialArc}.Radial Fill
    Threshold=@n13
   ) {materialFunction=&MF_UI_AntiAliasing}.Outer) {name=SegmentedMeterFill, nodeColor=<rgba 0.025 0.0203 0.0125 0.75>}
  n15 = Constant()
  n16 = Constant() {r=1}
  n17 = Add(
   A=@n10
   B=MaterialFunctionCall(Input=@n15, Target High=Divide(A=@n16, B=@n8)) {materialFunction=&RemapValueRange}.Result
  )
  n18 = LinearInterpolate(Alpha=Floor(None=@n15)) {constB=-1}
  n19 = MaterialFunctionCall(
   UV=NamedRerouteUsage()
   "Arc Start Angle (0-1)"=@n3
   "Arc End Angle (0-1)"=@n4
   "Fill Ratio (0-1)"=OneMinus(None=Add(A=@n17, B=@n18))
   Fill Clockwise=@n12
  ) {materialFunction=&MF_UI_RadialArc}
  n20 = Subtract(
   A=MaterialFunctionCall(DistanceField=@n19.Radial Fill, Threshold=@n13) {materialFunction=&MF_UI_AntiAliasing}.Outer
   B=@n14
  )
  n21 = Ceil(None=OneMinus(None=@n19.Arc Mask))
  n22 = Max(
   A=Max(A=@n14, B=NamedRerouteDeclaration(None=@n20) {name=SegmentedDeltaFill, nodeColor=<rgba 0.025 0.0203 0.0125 0.75>})
   B=Multiply(A=Saturate(None=@n21), B=ScalarParameter() {defaultValue=0.5, parameterName=BackgroundOpacity, sliderMax=1, sortPriority=4}) {constB=0.1}
  )
  n23 = MaterialFunctionCall(DistanceField=NamedRerouteUsage(), Threshold=Constant() {r=0.25}) {materialFunction=&MF_UI_AntiAliasing}
  n24 = MaterialFunctionCall(
   UVs=NamedRerouteUsage()
   Size=ScalarParameter() {defaultValue=2, group=Circle Parameters, parameterName=CircleSize, sliderMax=1, sortPriority=4}
   OutlineThickness=ScalarParameter() {defaultValue=0.05, group=Circle Parameters, parameterName=CircleThickness, sliderMax=1, sortPriority=4}
  ) {materialFunction=&MF_UI_Circle}
  Multiply(
   A=Multiply(
    A=MaterialFunctionCall(DistanceField=@n24.SDFOutline) {materialFunction=&MF_UI_AntiAliasing}.Inner
    B=Multiply(A=@n22, B=@n23.Inner)
   )
   B=LinearInterpolate(
    A=ScalarParameter() {defaultValue=0.8, parameterName=MasterOpacity, sliderMax=1}
    B=Constant() {r=0.4}
    Alpha=ScalarParameter() {parameterName=isCooldown}
   )
  )
  NamedRerouteDeclaration(None=VectorParameter() {parameterName=ProgressColor, defaultValue=<rgba 0.0781 0.7836 1 1>}.RGB) {name=ProgressColor, nodeColor=<rgba 0.025 0.0203 0.0125 0.75>}
  n25 = Multiply(A=@n19.Arc Mask, B=NamedRerouteDeclaration(None=Floor(None=@n8)) {name=Sector Count, nodeColor=<rgba 0 0 0 1>}) {desc=0 to Sector Count}
  n26 = Constant() {r=1}
  n27 = OneMinus(None=NamedRerouteDeclaration(None=MaterialFunctionCall!(Value=@n25, Scale=@n26, Height) {materialFunction=/FortUILibrary/MaterialFunctions/Basics/MF_UI_PingPong}.Result) {name=Segmented Gradient, nodeColor=<rgba 0 0 0 1>})
  n28 = OneMinus(None=NamedRerouteUsage())
  NamedRerouteDeclaration(None=MaterialFunctionCall(Input=Subtract(A=@n27, B=@n28) {constB=0}, Input High=NamedRerouteUsage()) {materialFunction=&RemapValueRange}.Result) {name=Convergent Sectors, nodeColor=<rgba 0 0 0 1>}
  n29 = LinearInterpolate(A=ScalarParameter() {defaultValue=0.5, group=Segmented Parameters, parameterName=SegmentSpacing, sliderMax=1, sortPriority=5}, Alpha=@n7) {constA=0.2, constB=0, desc=Switch for Segment Spacing}
  n30 = OneMinus(None=NamedRerouteDeclaration(None=OneMinus(None=@n29)) {name=Segment Spacing, nodeColor=<rgba 0.025 0.0203 0.0125 0.75>})
  n31 = Multiply(A=Multiply(A=NamedRerouteUsage(), B=NamedRerouteUsage())) {constB=0.25}
  NamedRerouteDeclaration(None=LinearInterpolate(
    A=NamedRerouteDeclaration(None=@n30) {name=Proportional Spacing, nodeColor=<rgba 0 0 0 1>}
    B=NamedRerouteDeclaration(None=@n31) {name=Fixed Spacing, nodeColor=<rgba 0 0 0 1>}
    Alpha=Constant()
   )) {name=Sector Spacing, nodeColor=<rgba 0 0 0 1>}
  n32 = Divide(
   A=ComponentMask(None=VectorParameter() {parameterName=CameraVelocity, defaultValue=<rgba 1 1 1 1>}.RGB) {g=true, r=true}
   B=Constant2Vector() {g=-10, r=-10}
  ) {desc="Decreases movement\r\nof element relative to others"}
  n33 = Constant2Vector() {desc=Max Drift, g=10, r=0.9}
  n34 = Multiply(A=@n33) {constB=-1}
  NamedRerouteDeclaration(None=Add(
    A=MaterialFunctionCall() {materialFunction=&GetUserInterfaceUV}.Normalized UV
    B=Clamp(None=@n32, Min=@n34, Max=@n33)
   )) {name=ParallaxUV, nodeColor=<rgba 0 0 0 1>}
  NamedRerouteDeclaration(None=VectorParameter() {parameterName=DeltaColor, defaultValue=<rgba 1 0.0014 0.0358 1>}.RGB) {name=DeltaColor, nodeColor=<rgba 0.025 0.0203 0.0125 0.75>}
  n35 = MaterialFunctionCall() {materialFunction=&GetUserInterfaceUV}
  n36 = NamedRerouteUsage()
  n37 = ScalarParameter() {parameterName=isParallax}
  NamedRerouteDeclaration(None=Add(A=LinearInterpolate(A=@n35.Normalized UV, B=@n36, Alpha=@n37), B=Constant2Vector() {r=0.75})) {name=ModUvs, nodeColor=<rgba 0.5 0.4065 0.25 0.75>}
 ]
}
