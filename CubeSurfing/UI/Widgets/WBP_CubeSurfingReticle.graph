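# Weapon reticle widget for the Cube Surfing weapon, showing a center dot, an ammo-driven fuel meter and a boost meter shown while surfing.
# On construct it binds weapon ammo changes, gameplay tag waits and boost duration, cooldown and failure events.
# Looping timers drive material parameters for boost drain and recharge and fuel refill near cooldown end, with ready sounds.
# A hotfix-gated camera rotation effects object adds parallax and pitch, updated by timer during surfing.
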WidgetBlueprint /CubeSurfing/UI/Widgets/WBP_CubeSurfingReticle
animations [
 Anim_BoostFail {
  binding_count=1
  length_seconds=5
  object=&self:Anim_BoostFail
  sequence {bindings [Image_BoostMeter {
     id=#1
     possessed_class=Image
     tracks [{
       type=MovieSceneWidgetMaterialTrack
       sections [{
         type=MovieSceneParameterSection
         end_frame=100
         start_frame=0
         channels [isError {type=MovieSceneScriptingFloatChannel, keys [[0,0],[5.0253,1],[9.871,0]]}]
        }]
      }]
    }]}
 }
 Anim_Initialize {
  binding_count=1
  length_seconds=0.45
  object=&self:Anim_Initialize
  sequence {bindings [Image_FuelMeter {
     id=#2
     possessed_class=Image
     tracks [{
       type=MovieSceneColorTrack
       display_name=Color and Opacity
       sections [{
         type=MovieSceneColorSection
         end_frame=9
         start_frame=0
         channels [
          Color.R {type=MovieSceneScriptingFloatChannel, const [1,3]}
          Color.G {type=MovieSceneScriptingFloatChannel, const [1,3]}
          Color.B {type=MovieSceneScriptingFloatChannel, const [1,3]}
          Color.A {type=MovieSceneScriptingFloatChannel, keys [[0,0],[4,0.0253],[6,0.2469],[9,1]]}
         ]
        }]
      }]
    }]}
 }
]
graphs [
 EventGraph {nodes [
   UpdateParallax() {is=title K2Node_CustomEvent}
   > Update!(
    execute
    self:Fort Widget Camera Rotation Effects Object Reference=Get CameraRotationEffects() -> (CameraRotationEffects:Fort Widget Camera Rotation Effects Object Reference)
    DeltaTime:float=Get TimerUpdateRate() -> (TimerUpdateRate:double)
   ) -> (then)
   n1 = LocalAmmoChanged!() -> (OutputDelegate, then, LocalCount:int, LocalRemaining:int) {is=title K2Node_CustomEvent}
   > Set Local Remaining(Local Remaining:int=@n1.LocalRemaining) -> (Output_Get:int=0)
   > n2 = Set Local Count(Local Count:int=@n1.LocalCount) -> (Output_Get:int=0)
   > n3 = GetDynamicMaterial(self=Get Image_FuelMeter() -> (Image_FuelMeter:Image Object Reference))
   > SetScalarParameterValue(self=@n3, ParameterName=ActiveAmmo, Value="To Float (Integer)"(InInt=@n2))
   Event Construct()
   > n4 = Sequence!(execute) -> (then_0, then_1, then_2:Exec, then_3:Exec, then_4:Exec, then_5:Exec, then_6:Exec)
   > n5 = Cast To B_CubeSurfing_Weapon!(
    execute
    Object=GetEquippedWeapon!(self:Fort Weapon Reticle Extension Widget Base Object Reference) -> (ReturnValue:Fort Weapon Object Reference)
   ) -> (then, CastFailed, AsB Cube Surfing Weapon:B Cube Surfing Weapon Object Reference) {is=title K2Node_DynamicCast}
   > n6 = Set As B Cube Surfing Weapon(As B Cube Surfing Weapon:B Cube Surfing Weapon Object Reference=@n5) -> (Output_Get:B Cube Surfing Weapon Object Reference)
   > Bind Event to On Local Ammo Changed(self:Fort Weapon Object Reference=@n6, Delegate=Create Event()) {is=title K2Node_AddDelegate}
   > Set Local Count(Local Count=GetLocalMagazineAmmoCount!(self:Fort Weapon Object Reference=Get As B Cube Surfing Weapon() -> (As B Cube Surfing Weapon:B Cube Surfing Weapon Object Reference)) -> (ReturnValue:int=0)) -> (Output_Get=0)
   > n7 = Set PlayerController(PlayerController:Player Controller Object Reference=GetOwningPlayer()) -> (Output_Get:Player Controller Object Reference)
   > n8 = Cast To FortPlayerController(Object=@n7)
   > Set PlayerPawn(PlayerPawn:Fort Player Pawn Object Reference=GetPlayerPawn!(self:Fort Player Controller Object Reference=@n8) -> (ReturnValue:Fort Player Pawn Object Reference)) -> (Output_Get:Fort Player Pawn Object Reference)
   > n9 = GetDynamicMaterial(self=Get Image_FuelMeter())
   > n10 = Set Dynamic_FuelMeter(Dynamic_FuelMeter:Material Instance Dynamic Object Reference=@n9) -> (Output_Get:Material Instance Dynamic Object Reference)
   > SetScalarParameterValue(self=@n10, ParameterName=ActiveAmmo, Value="To Float (Integer)"(InInt=Get Local Count() -> (Local Count:int=0)))
   > n11 = GetDynamicMaterial(self=Get Image_BoostMeter() -> (Image_BoostMeter:Image Object Reference))
   > Set Dynamic_BoostMeter(Dynamic_BoostMeter:Material Instance Dynamic Object Reference=@n11) -> (Output_Get:Material Instance Dynamic Object Reference)
   > n12 = Set ClipSize(ClipSize:int=GetBulletsPerClip!(self:Fort Weapon Object Reference=Get As B Cube Surfing Weapon()) -> (ReturnValue:int=0)) -> (Output_Get:int=0)
   > SetScalarParameterValue(
    self=Get Dynamic_FuelMeter() -> (Dynamic_FuelMeter:Material Instance Dynamic Object Reference)
    ParameterName=ClipSize
    Value="To Float (Integer)"(InInt=@n12)
   )
   n13 = WaitGameplayTagAddToActor!(
    execute:Exec=@n4.then_1
    TargetActor:Actor Object Reference=Get PlayerPawn() -> (PlayerPawn:Fort Player Pawn Object Reference)
    Tag:Gameplay Tag Structure=(TagName="Granted.CubeSurfing.Active")
    OnlyTriggerOnce:bool=false
   ) -> (then:Exec, AsyncTaskProxy:Ability Async Wait Gameplay Tag Added Object Reference, Added:Exec) {is=title K2Node_AsyncAction}
   > n14 = WaitGameplayTagRemoveFromActor!(
    execute:Exec
    TargetActor:Actor Object Reference=Get PlayerPawn()
    Tag:Gameplay Tag Structure=(TagName="Granted.CubeSurfing.Active")
    OnlyTriggerOnce:bool=false
   ) -> (then:Exec, AsyncTaskProxy:Ability Async Wait Gameplay Tag Removed Object Reference, Removed:Exec) {is=title K2Node_AsyncAction}
   SetVisibility(execute=@n14.Removed, self=Get Image_BoostMeter(), InVisibility=Hidden)
   SetVisibility(execute=@n13.Added, self=Get Image_BoostMeter(), InVisibility=Visible)
   n15 = StartListeningToEvent!(
    execute:Exec=@n4.then_2
    Target:Gameplay Event Router Component Object Reference=GetGlobalEventRouter()
    EventType:Script Struct Object Reference=/Script/CubeSurfingRuntime.CubeSurfingEvent_BoostDuration
    Context:Object Reference=Get PlayerPawn()
    EventBubblingRule:EEventBubblingRule Enum
   ) -> (
    then:Exec
    AsyncTaskProxy:Async Action Start Listening To Event Object Reference
    OnEventReceived:Exec
    Payload:Cube Surfing Event Boost Duration Structure
   )
   Set BoostTimeElapsed(execute=@n15.OnEventReceived, BoostTimeElapsed:float=0) -> (Output_Get:float)
   > Set BoostTimeLength(BoostTimeLength:float=Break Cube Surfing Event Boost Duration!(CubeSurfingEvent_BoostDuration:"Cube Surfing Event Boost Duration Structure (by ref)"=@n15.Payload) -> (MaxDuration:float=~) {is=title K2Node_BreakStruct}) -> (Output_Get:float)
   > n16 = Set Timer by Event(Delegate=Create Event(), Time=Get TimerUpdateRate(), bLooping=true)
   > Set BoostTimerHandle(BoostTimerHandle:Timer Handle Structure=@n16) -> (Output_Get:Timer Handle Structure)
   n17 = WaitGameplayTagAddToActor(execute=@n4.then_3, TargetActor=Get PlayerPawn(), Tag=(TagName="Granted.Cooldown.CubeSurfing.Primary"), OnlyTriggerOnce=false)
   n18 = GetCooldownTimeRemainingAndDuration!(
    execute=@n17.Added
    Actor:Actor Object Reference=Get PlayerPawn()
    CooldownTags:Gameplay Tag Container Structure=(GameplayTags=((TagName="Ability"),(TagName="Granted.Cooldown.CubeSurfing.Primary")))
   ) -> (then, TimeRemaining:float, CooldownDuration:float)
   > Set Cooldown Duration(Cooldown Duration:float=@n18.CooldownDuration) -> (Output_Get:float)
   > SetScalarParameterValue(self=Get Dynamic_FuelMeter(), ParameterName=isCooldown, Value=1)
   > n19 = Set Timer by Event(Delegate=Create Event(), Time=Get TimerUpdateRate(), bLooping=true)
   > Set CooldownTimerHandle(CooldownTimerHandle:Timer Handle Structure=@n19) -> (Output_Get:Timer Handle Structure)
   n20 = StartListeningToEvent(
    execute=@n4.then_4
    Target=GetGlobalEventRouter()
    EventType=/Script/CubeSurfingRuntime.CubeSurfingEvent_BoostCooldown
    Context=Get PlayerPawn()
   ) -> (Payload:Cube Surfing Event Boost Cooldown Structure)
   Clear and Invalidate Timer by Handle(execute=@n20.OnEventReceived, Handle=Get BoostTimerHandle() -> (BoostTimerHandle:Timer Handle Structure))
   > Set BoostTimeElapsed(BoostTimeElapsed=0)
   > n21 = GetDynamicMaterial(self=Get Image_BoostMeter())
   > SetScalarParameterValue(self=@n21, ParameterName=isCooldown, Value=1)
   > Set BoostCooldownTimeElapsed(BoostCooldownTimeElapsed:double=0) -> (Output_Get:double)
   > Set Boost Cooldown Duration(Boost Cooldown Duration:float=Break Cube Surfing Event Boost Cooldown!(CubeSurfingEvent_BoostCooldown:"Cube Surfing Event Boost Cooldown Structure (by ref)"=@n20.Payload) -> (CooldownLength:float=~) {is=title K2Node_BreakStruct}) -> (Output_Get:float)
   > n22 = Set Timer by Event(Delegate=Create Event(), Time=Get TimerUpdateRate(), bLooping=true)
   > Set BoostCooldownTimerHandle(BoostCooldownTimerHandle:Timer Handle Structure=@n22) -> (Output_Get:Timer Handle Structure)
   n23 = StartListeningToEvent(
    execute=@n4.then_5
    Target=GetGlobalEventRouter()
    EventType=/Script/CubeSurfingRuntime.CubeSurfingEvent_BoostFailed
    Context=Get PlayerPawn()
   ) -> (Payload:Cube Surfing Event Boost Failed Structure)
   PlayAnimation(execute=@n23.OnEventReceived, InAnimation=Get Anim_BoostFail() -> (Anim_BoostFail:Widget Animation Object Reference))
   Branch(
    execute=@n4.then_6
    Condition=GetValueAsBool!(Input:"Scalable Float Structure (by ref)"=Get CubeSurfParallaxHotfix() -> (CubeSurfParallaxHotfix:Scalable Float Structure), Level:float=0) -> (ReturnValue:bool)
   )
   > SetScalarParameterValue(self=Get Dynamic_BoostMeter() -> (Dynamic_BoostMeter:Material Instance Dynamic Object Reference), ParameterName=isParallax, Value=1)
   > SetScalarParameterValue(self=Get Dynamic_FuelMeter(), ParameterName=isParallax, Value=1)
   > n24 = CreateCameraRotationEffectsObject!(execute, Outer:Object Reference) -> (then, ReturnValue:Fort Widget Camera Rotation Effects Object Reference)
   n25 = Get CameraRotationEffects()
   > "Set Fort Widget Camera Rotation Effects Object Reference (by ref)"!(
    execute:Exec
    Target:"Fort Widget Camera Rotation Effects Object Reference (by ref)"=@n25
    Value:Fort Widget Camera Rotation Effects Object Reference=@n24
   ) -> (then:Exec) {is=title K2Node_VariableSetRef}
   n26 = Get Dynamic_FuelMeter()
   n27 = Get Dynamic_BoostMeter()
   > Setup!(
    execute
    self:Fort Widget Camera Rotation Effects Object Reference=@n25
    InPlayerCameraManager:Player Camera Manager Object Reference=Get PlayerCameraManager!(self:Player Controller Object Reference=GetOwningPlayer()) -> (PlayerCameraManager:Player Camera Manager Object Reference)
    InParallaxEffectData:Array of Parallax Effect Data Structures=Make Array!(
     "[0]":Parallax Effect Data Structure=Make Parallax Effect Data!(
      PitchOffsetMultiplier:float=1
      YawOffsetMultiplier:float=-1
      Material:Material Instance Dynamic Object Reference=@n26
      ParameterName:Name=CameraVelocity
     ) -> (ParallaxEffectData:Parallax Effect Data Structure) {is=title K2Node_MakeStruct}
     "[1]":Parallax Effect Data Structure=Make Parallax Effect Data(PitchOffsetMultiplier=1, YawOffsetMultiplier=-1, Material=@n27, ParameterName=CameraVelocity)
    ) -> (Array:Array of Parallax Effect Data Structures)
    InPitchMeterEffectData:"Basic Effect Data Structure (by ref)"=Make Basic Effect Data!(Material:Material Instance Dynamic Object Reference=@n27, ParameterName:Name=Pitch) -> (BasicEffectData:Basic Effect Data Structure) {is=title K2Node_MakeStruct}
    InMinMaxPitchRate:float=0.6
    InMinMaxYawRate:float=0.8
    InOffsetInterpSpeed:float=2.5
    InResetInterpSpeed:float=5
    InVelocityMultiplier:float=0.8
   ) -> (then)
   > Update(self=Get CameraRotationEffects(), DeltaTime=0)
   > n28 = WaitGameplayTagAddToActor(TargetActor=Get PlayerPawn(), Tag=(TagName="Granted.CubeSurfing.Active"), OnlyTriggerOnce=false)
   > n29 = WaitGameplayTagRemoveFromActor(TargetActor=Get PlayerPawn(), Tag=(TagName="Granted.CubeSurfing.Active"), OnlyTriggerOnce=false)
   Clear and Invalidate Timer by Handle(execute=@n29.Removed, Handle=Get ParallaxTimerHandle() -> (ParallaxTimerHandle:Timer Handle Structure))
   n30 = Set Timer by Event(execute=@n28.Added, Delegate=Create Event(), Time=Get TimerUpdateRate(), bLooping=true)
   > Set ParallaxTimerHandle(ParallaxTimerHandle:Timer Handle Structure=@n30) -> (Output_Get:Timer Handle Structure)
   Event Destruct()
   > Unbind Event from On Local Ammo Changed(self:Fort Weapon Object Reference=Get As B Cube Surfing Weapon(), Delegate=Create Event()) {is=title K2Node_RemoveDelegate}
   Event OnInitialized()
   > PlayAnimation(InAnimation=Get Anim_Initialize() -> (Anim_Initialize:Widget Animation Object Reference))
  ]}
 OnBoostCooldownUpdate {nodes [
   OnBoostCooldownUpdate() {is=title K2Node_FunctionEntry}
   > n1 = Branch(Condition="float >= float"(A=Get BoostCooldownTimeElapsed() -> (BoostCooldownTimeElapsed:double), B=Get Boost Cooldown Duration() -> (Boost Cooldown Duration:float)))
   > Clear and Invalidate Timer by Handle(Handle=Get BoostCooldownTimerHandle() -> (BoostCooldownTimerHandle:Timer Handle Structure))
   > Set BoostCooldownTimeElapsed(BoostCooldownTimeElapsed=0)
   > SetScalarParameterValue(self=Get Dynamic_BoostMeter(), ParameterName=isCooldown, Value=0)
   > Set BoostTimeElapsed(BoostTimeElapsed=0)
   > SetScalarParameterValue(self=Get Dynamic_BoostMeter(), ParameterName=MeterProgress, Value=1)
   > StopAnimation(InAnimation=Get Anim_BoostFail())
   > PlayEvent!(
    execute
    self:Sound Library Component Object Reference=GetComponentByClass!(
     self:Actor Object Reference=Get As B Cube Surfing Weapon()
     ComponentClass:Actor Component Class Reference=/Game/Sounds/Fort_Audio_Blueprints/SpatialTravelAudioComponent/SpatialTravel_SoundLibraryComponent.SpatialTravel_SoundLibraryComponent_C
    ) -> (ReturnValue:Spatial Travel Sound Library Component Object Reference)
    InEventName:Gameplay Tag Structure=(TagName="SoundLibrary.SpatialTravel.Boost.Ready")
   ) -> (then, ReturnValue:Sound Library Play Result Structure)
   n2 = Get Boost Cooldown Duration()
   n3 = Get BoostCooldownTimeElapsed()
   SetScalarParameterValue(execute=@n1.else, self=Get Dynamic_BoostMeter(), ParameterName=MeterProgress, Value=Lerp(A=0, B=1, Alpha=float / float(A=@n3, B=@n2)))
   > Set BoostCooldownTimeElapsed(BoostCooldownTimeElapsed=float + float(A=Get BoostCooldownTimeElapsed(), B=Get TimerUpdateRate()))
  ]}
 OnBoostUpdate {nodes [
   OnBoostUpdate() {is=title K2Node_FunctionEntry}
   > n1 = Branch(Condition="float >= float"(A=Get BoostTimeElapsed() -> (BoostTimeElapsed:float), B=Get BoostTimeLength() -> (BoostTimeLength:float)))
   > Clear and Invalidate Timer by Handle(Handle=Get BoostTimerHandle())
   n2 = Get BoostTimeLength()
   n3 = Get BoostTimeElapsed()
   SetScalarParameterValue(execute=@n1.else, self=Get Dynamic_BoostMeter(), ParameterName=MeterProgress, Value=Lerp(A=1, B=0, Alpha=float / float(A=@n3, B=@n2)))
   > Set BoostTimeElapsed(BoostTimeElapsed=float + float(A=Get BoostTimeElapsed(), B=Get TimerUpdateRate()))
  ]}
 OnCooldownUpdate {nodes [
   OnCooldownUpdate() {is=title K2Node_FunctionEntry}
   > n1 = Branch(Condition="float >= float"(A=Get CooldownTimeElapsed() -> (CooldownTimeElapsed:double), B=Get Cooldown Duration() -> (Cooldown Duration:float)))
   > Clear and Invalidate Timer by Handle(Handle=Get CooldownTimerHandle() -> (CooldownTimerHandle:Timer Handle Structure))
   > Set CooldownTimeElapsed(CooldownTimeElapsed:double=0) -> (Output_Get:double)
   > Branch(Condition=IsVisible(self=Get Image_FuelMeter()))
   > PlaySound2D(Sound=/CubeSurfing/Assets/Audio/MSS_CubeSurfing_Surf_Ready_1P)
   n2 = Get Cooldown Duration()
   n3 = Get CooldownTimeElapsed()
   n4 = Branch(
    execute=@n1.else
    Condition="float <= float"(A=float - float(A=@n2, B=@n3), B=Get AmmoRefillAnimationTime() -> (AmmoRefillAnimationTime:double)) -> (ReturnValue=~)
   )
   n5 = "Min (Integer)"(A=Get Local Remaining() -> (Local Remaining:int=0), B=Get ClipSize() -> (ClipSize:int=0)) -> (ReturnValue=0)
   n6 = Get Local Count() -> (Local Count=0)
   n7 = float / float(A=float - float(A=Get Cooldown Duration(), B=Get CooldownTimeElapsed()), B=Get AmmoRefillAnimationTime())
   n8 = Get Dynamic_FuelMeter()
   > SetScalarParameterValue(self=@n8, ParameterName=ActiveAmmo, Value=Lerp(A="To Float (Integer)"(InInt=@n5), B="To Float (Integer)"(InInt=@n6), Alpha=@n7))
   > SetScalarParameterValue(self=@n8, ParameterName=isCooldown, Value=@n7)
   > Set CooldownTimeElapsed(CooldownTimeElapsed=float + float(A=Get CooldownTimeElapsed(), B=Get TimerUpdateRate()))
   Set CooldownTimeElapsed(execute=@n4.else, CooldownTimeElapsed=float + float(A=Get CooldownTimeElapsed(), B=Get TimerUpdateRate()))
  ]}
]
variables [
 As B Cube Surfing Weapon {type=B_CubeSurfing_Weapon_C, default=null}
 Boost Cooldown Duration {type=number}
 BoostCooldownTimeElapsed {type=number}
 BoostCooldownTimerHandle {type=TimerHandle}
 BoostTimeElapsed {type=number}
 BoostTimeLength {type=number}
 BoostTimerHandle {type=TimerHandle}
 CameraRotationEffects {type=FortWidgetCameraRotationEffects, default=null}
 ClipSize {type=integer}
 Cooldown Duration {type=number}
 CooldownTimeElapsed {type=number}
 CooldownTimerHandle {type=TimerHandle}
 CubeSurfParallaxHotfix {
  type=ScalableFloat
  default {registryType=None, value=1, curve {curveTable=/CubeSurfing/DataTables/CubeSurfingGameData, rowName=CubeSurf.ReticleParallax}}
 }
 Dynamic_BoostMeter {type=MaterialInstanceDynamic}
 Dynamic_FuelMeter {type=MaterialInstanceDynamic}
 Local Count {type=integer}
 Local Remaining {type=integer}
 Parallax Effect Data {type="array<ParallaxEffectData>", default []}
 ParallaxTimerHandle {type=TimerHandle}
 PlayerController {type=PlayerController, default=null}
 PlayerPawn {type=FortPlayerPawn}
 TimerUpdateRate {type=number, default=0.01}
]
widgets {
 info {
  inheritedWidgetCount=1
  parentClass=/Script/FortniteUI.FortChargedWeaponReticleExtensionWidget
  rootWidgetClass=&Overlay
  widgetCount=7
 }
 widgets [
  Overlay_39 {widgetClassPath=&Overlay, widget=Overlay_39}
  CenterDot {
   widgetClassPath=&Image
   parent=Overlay_39
   widget=CenterDot
   properties {brush {
     bIsDynamicallyLoaded=false
     drawAs=Image
     imageType=FullColor
     mirroring=NoMirror
     resourceName=None
     resourceObject=/CubeSurfing/UI/Materials/M_UI_CubeSurfingCenterReticle
     tiling=NoTile
     imageSize=<16 16>
     margin=<ltrb 0 0 0 0>
     outlineSettings {
      bUseBrushTransparency=false
      roundingType=HalfHeightRadius
      width=0
      color {colorUseRule=UseColor_Specified, specifiedColor=<rgba 0 0 0 0>}
      cornerRadii=<0 0 0 0>
     }
     tintColor {colorUseRule=UseColor_Specified, specifiedColor=<rgba 1 1 1 1>}
     uVRegion {bIsValid=false, max=<0 0>, min=<0 0>}
    }}
   slot {class=OverlaySlot, properties {horizontalAlignment=HAlign_Center, verticalAlignment=VAlign_Center}}
  }
  FortVisualAttachment_0 {
   widgetClassPath=&FortVisualAttachment
   parent=Overlay_39
   widget=FortVisualAttachment_0
   slot {class=OverlaySlot, properties {horizontalAlignment=HAlign_Center, verticalAlignment=VAlign_Center}}
  }
  Image_FuelMeter {
   widgetClassPath=&Image
   bIsVariable=true
   parent=FortVisualAttachment_0
   widget=Image_FuelMeter
   properties {
    brush {
     bIsDynamicallyLoaded=false
     drawAs=Image
     imageType=FullColor
     mirroring=NoMirror
     resourceName=None
     resourceObject=/CubeSurfing/UI/Materials/M_UI_CubeSurfingFuelMeter
     tiling=NoTile
     imageSize=<200 270>
     margin=<ltrb 0 0 0 0>
     outlineSettings {
      bUseBrushTransparency=false
      roundingType=HalfHeightRadius
      width=0
      color {colorUseRule=UseColor_Specified, specifiedColor=<rgba 0 0 0 0>}
      cornerRadii=<0 0 0 0>
     }
     tintColor {colorUseRule=UseColor_Specified, specifiedColor=<rgba 1 1 1 1>}
     uVRegion {bIsValid=false, max=<0 0>, min=<0 0>}
    }
    renderTransform {angle=0, scale=<1 1>, shear=<0 0>, translation=<94 -135>}
   }
   slot {class=SizeBoxSlot}
  }
  FortVisualAttachment_1 {
   widgetClassPath=&FortVisualAttachment
   parent=Overlay_39
   widget=FortVisualAttachment_1
   slot {class=OverlaySlot, properties {horizontalAlignment=HAlign_Center, verticalAlignment=VAlign_Center}}
  }
  Image_BoostMeter {
   widgetClassPath=&Image
   bIsVariable=true
   parent=FortVisualAttachment_1
   widget=Image_BoostMeter
   properties {
    brush {
     bIsDynamicallyLoaded=false
     drawAs=Image
     imageType=FullColor
     mirroring=NoMirror
     resourceName=None
     resourceObject=/CubeSurfing/UI/Materials/M_UI_CubeSurfingBoostMeter
     tiling=NoTile
     imageSize=<250 465>
     margin=<ltrb 0 0 0 0>
     outlineSettings {
      bUseBrushTransparency=false
      roundingType=HalfHeightRadius
      width=0
      color {colorUseRule=UseColor_Specified, specifiedColor=<rgba 0 0 0 0>}
      cornerRadii=<0 0 0 0>
     }
     tintColor {colorUseRule=UseColor_Specified, specifiedColor=<rgba 1 1 1 1>}
     uVRegion {bIsValid=false, max=<0 0>, min=<0 0>}
    }
    renderTransform {angle=0, scale=<1 1>, shear=<0 0>, translation=<48 -232.5>}
   }
   slot {class=SizeBoxSlot}
  }
  TutorialHighlight {widgetClassPath=&TutorialHighlight}
 ]
}
